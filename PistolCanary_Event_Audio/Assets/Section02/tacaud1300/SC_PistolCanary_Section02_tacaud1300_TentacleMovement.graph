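# Sound cue for tentacle movement in Section02 of the PistolCanary event audio.
# A single wave player node feeds the output directly, with no randomization or mixing, and the file shows no wave assigned to it.
# The cue uses the event's tentacle attenuation asset for long-range falloff, routes to the Fort_Events sound class and plays at reduced volume.
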SoundCue /PistolCanary_Event_Audio/Assets/Section02/tacaud1300/SC_PistolCanary_Section02_tacaud1300_TentacleMovement
root=WavePlayer_0
nodes [{id=WavePlayer_0}]
properties {
 attenuationSettings=/PistolCanary_Event_Audio/Assets/Settings/ATT_PistolCanary_Tentacle
 maxDistance=23000
 soundClassObject=&Fort_Events
 volumeMultiplier=0.6627
}
