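# Sound cue for a tentacle death in Section 02 of the PistolCanary event, in the Fort_Events sound class.
# A mixer plays the attack-death wave immediately alongside a second death wave held back by a fixed one-and-a-half-second delay.
# It uses the event's tentacle-death attenuation asset and sends at a low level to a weapon-volume sidechain audio bus.
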SoundCue /PistolCanary_Event_Audio/Assets/Section02/tacaud1300/SC_PistolCanary_Section02_tacaud1300_Tentacle_Death
root=Mixer_0
nodes [
 {id=Delay_0, children [WavePlayer_0], properties {delayMax=1.5, delayMin=1.5}}
 {id=Mixer_0, children [WavePlayer_1,Delay_0], properties {inputVolume [1,1]}}
 {id=WavePlayer_0, properties {soundWaveAssetPtr=/PistolCanary_Event_Audio/Assets/Section02/wav/PistolCanary_Sec_03_Tentacle_Death_01}}
 {id=WavePlayer_1, properties {soundWaveAssetPtr=/PistolCanary_Event_Audio/Assets/Section02/wav/PistolCanary_Tentacle_Attk_1300_Death}}
]
properties {
 attenuationSettings=/PistolCanary_Event_Audio/Assets/Settings/ATT_PistolCanary_Tentacle_Deaths
 maxDistance=35000
 soundClassObject=&Fort_Events
 volumeMultiplier=1
 busSends [{audioBus=&AB_Sidechain_FortWeapon_Vol, minSendLevel=0, sendLevel=0.2, soundSourceBus=null}]
}
